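# Conversation database for the Agent Jones NPC, defining the greeting and conversation branches a player gets when talking to him.
# A single entry point leads to a speech node whose default line can be overridden by first-meeting or outfit-reaction messages, prioritized through curve rows.
# Three of the five outfit reactions repeat the default line.
# The speech node links out by tag to shared NPC quests, basic services, and Season 33 weapon expertise tutorial and SMG-completion conversations.
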ConversationDatabase /NPCLibrary/NPCs/AgentJones/AgentJones/NPCLibrary_Conversation_AgentJones
entry_points [{tag=Conversation.Entry.AgentJones, destinations [#1]}]
nodes [
 ConversationEntryPointNode_7 {
  id=#1
  class=ConversationEntryPointNode
  outputs [#2]
  properties {entryTag {tagName=Conversation.Entry.AgentJones}}
 }
 Speech_C_4 {
  id=#2
  class=Speech_C
  outputs [#3, #4, #5, #6]
  properties {
   nodeName=""
   generalConfig {
    defaultMessage=So I hear we have a time-travelling vampire problem now? Well that's just... super...
    contextualMessages [
     {
      message="Hey! I'm back!"
      contextRequirements [{requirement {class=FortControllerRequirement_IsFirstEverConversationWithNPC}}]
      priority {curve {rowName=Tandem.Speech.Priority.FirstEverInteraction}}
     }
     {
      message="I don't know what's more dangerous, you shooting at me, or you talking to me."
      requirementMatchPolicy=RequireAny
      contextRequirements [{requirement {class=FortControllerRequirement_HasCID, properties {allowedCIds [&Character_InnovatorSand]}}}]
      priority {curve {rowName=Tandem.Speech.Priority.OutfitReaction}}
     }
     {
      message="Keep counting, birdie. I think we can fix this."
      requirementMatchPolicy=RequireAny
      contextRequirements [{requirement {
         class=FortControllerRequirement_HasCID
         properties {allowedCIds [/BRCosmetics/Athena/Items/Cosmetics/Characters/CID_359_Athena_Commando_F_Aztec]}
        }}]
      priority {curve {rowName=Tandem.Speech.Priority.OutfitReaction}}
     }
     {
      message=So I hear we have a time-travelling vampire problem now? Well that's just... super...
      requirementMatchPolicy=RequireAny
      contextRequirements [{requirement {
         class=FortControllerRequirement_HasCID
         properties {allowedCIds [&Character_MusketSlinger, /BRCosmetics/Athena/Items/Cosmetics/Characters/Character_BlazerVeil]}
        }}]
      priority {curve {rowName=Tandem.Speech.Priority.OutfitReaction}}
     }
     {
      message=So I hear we have a time-travelling vampire problem now? Well that's just... super...
      requirementMatchPolicy=RequireAny
      contextRequirements [{requirement {class=FortControllerRequirement_HasCID, properties {allowedCIds [/BRCosmetics/Athena/Items/Cosmetics/Characters/Character_Lilac]}}}]
      priority {curve {rowName=Tandem.Speech.Priority.OutfitReaction}}
     }
     {
      message=So I hear we have a time-travelling vampire problem now? Well that's just... super...
      requirementMatchPolicy=RequireAny
      contextRequirements [{requirement {
         class=FortControllerRequirement_HasCID
         properties {allowedCIds [&CID_349_Athena_Commando_M_Banana]}
        }}]
      priority {curve {rowName=Tandem.Speech.Priority.OutfitReaction}}
     }
    ]
   }
  }
 }
 ConversationLinkNode_3 {
  id=#6
  class=ConversationLinkNode
  properties {remoteEntryTag {tagName=Conversation.Entry.Season33.Weaponexpertise.Tutorial.01}}
 }
 ConversationLinkNode_2 {
  id=#5
  class=ConversationLinkNode
  properties {remoteEntryTag {tagName=Conversation.Entry.Season33.Weaponexpertise.SMG.Complete.01}}
 }
 ConversationLinkNode_0 {id=#4, class=ConversationLinkNode, properties {remoteEntryTag {tagName=Conversation.Entry.BasicServices}}}
 ConversationLinkNode_1 {id=#3, class=ConversationLinkNode, properties {remoteEntryTag {tagName=Conversation.Entry.NPCLibrary.Quests}}}
]
